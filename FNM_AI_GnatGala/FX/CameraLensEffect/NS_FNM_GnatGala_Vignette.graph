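# Niagara camera lens effect that draws a vignette for the FNM Gnat Gala AI.
# The active Minimal emitter burst-spawns a black sprite with fixed facing, sized by a system Base Scale variable.
# System Timeline fades its alpha, while system-age curves animate dynamic material parameters.
# A disabled PP_Vignette emitter holds a component-renderer variant with timeline-driven blend weight and animated color, radius and position.
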NiagaraSystem /FNM_AI_GnatGala/FX/CameraLensEffect/NS_FNM_GnatGala_Vignette
dependencies {
 usedDataInterfaces [&NiagaraDataInterfaceCurve, &NiagaraDataInterfaceColorCurve]
 usedDynamicInputs [&SimulationPosition, &Multiply_Vector2D]
 usedModules [
  &SystemState
  &Timeline
  &self:SystemSpawnScript.SystemScriptSource.SystemScriptGraph.NiagaraNodeAssignment_0.SetVariables_#1
 ]
}
emitters {
 Minimal {
  emitterUpdateScript [
   &EmitterState(Inactive Response=::NewEnumerator1, Life Cycle Mode=::NewEnumerator1, Loop Duration=1, UseLoopDelay=false)
   &SpawnBurst_Instantaneous(Use Spawn Probability=false, UseLoopCountLimit=true)
  ]
  particleSpawnScript [
   &InitializeParticle(
    Color=<rgba 0 0 0 2>
    Sprite Size Mode=::NewEnumerator4
    Sprite UV Mode=::NewEnumerator3
    UsePositionOffset=false
    Write Material Random=true
   )
   &SpriteFacingAndAlignment(Sprite Alignment=<0 0 1>, Sprite Facing=<-1 0 0>, WriteAlignment=true, WriteFacing=true)
  ]
  particleUpdateScript [
   &ParticleState(Kill Particles When Lifetime Has Elapsed=false, Let Infinitely Lived Particles Die When Emitter Deactivates=true)
   &ScaleColor(
    Scale Alpha=&FloatFromCurve(
     FloatCurve=DataInterface() {propertyValues {Curve {keys [<at 0 0 RCIM_Cubic>,<at 1 1 RCIM_Cubic>]}}}
     CurveIndex=$System.Timeline.NormalizedTime
    )
    ScaleA=true
    ScaleRGB=true
   ) {module=ScaleColor001}
   &DynamicMaterialParameters(
    Index 0 Param 1=3
    Index 0 Param 2=0.1
    Index 0 Param 3=1
    Index 0 Param 4=0.1
    Index 1 Param 1=&FloatFromCurve(
     FloatCurve=DataInterface() {propertyValues {Curve {keys [{interpMode=RCIM_Cubic, tangentMode=RCTM_Break, time=0, value=1}, {interpMode=RCIM_Cubic, tangentMode=RCTM_Break, time=1, value=0}]}}}
     CurveIndex=$Engine.System.Age
    )
    Index 1 Param 2=&FloatFromCurve(FloatCurve=DataInterface() {propertyValues {Curve {keys [<at 0 0>,<at 0.1 0.555>]}}}, CurveIndex=$Engine.System.Age)
    Module.Index0Param0Write=true
    Module.Index0Param1Write=true
    Module.Index0Param2Write=true
    Module.Index0Param3Write=true
    Module.Index1Param0Write=true
    Module.Index1Param1Write=true
    Module.Index1Param2Write=true
    Module.Index1Param3Write=true
    Write Parameter Index 1=true
   )
   /SetVariables(Particles.SpriteSize=$System.Base Scale) {isSetParameters=true}
  ]
  rendererClasses [&NiagaraSpriteRendererProperties]
  renderers [{rendererClass=&NiagaraSpriteRendererProperties}]
 }
 PP_Vignette {
  bEnabled=false
  emitterSpawnScript [/SetVariables(Emitter.Global Scale:NiagaraFloat=$User.Global Scale) {isSetParameters=true}]
  emitterUpdateScript [
   &EmitterState()
   &SpawnBurst_Instantaneous(Use Spawn Probability=false, UseLoopCountLimit=false) {module=SpawnBurst_Instantaneous002}
   /SetVariables(
    Emitter. Color:LinearColor=&ColorFromCurve(
     ColorCurve=DataInterface() {propertyValues {
       AlphaCurve {keys [<at 0.4432 1 RCIM_Cubic>,<at 0.9517 0 RCIM_Cubic>]}
       BlueCurve {keys [
         <at 0 0.599>
         {arriveTangentWeight=0.0456, interpMode=RCIM_Cubic, leaveTangentWeight=0.1363, tangentMode=RCTM_User, time=0.1352, value=0}
         <at 0.4679 0.125>
         <at 0.8087 0 RCIM_Cubic>
        ]}
       GreenCurve {keys [
         <at 0 1.9767>
         {arriveTangent=-2.7155, interpMode=RCIM_Cubic, leaveTangent=-2.7155, time=0.1352, value=7.7363}
         <at 0.4679 0.7062>
         <at 0.8087 0 RCIM_Cubic>
        ]}
       RedCurve {keys [
         <at 0 5>
         {arriveTangent=-2.1374, interpMode=RCIM_Cubic, leaveTangent=-2.1374, time=0.1352, value=50}
         <at 0.4679 4>
         <at 0.8087 0 RCIM_Cubic>
        ]}
      }}
     CurveIndex=$Emitter.NormalizedLoopAge
    )
    Emitter.Exponent=1.8
    Emitter.LightEnabled:NiagaraBool=&SetBoolByFloatComparison(A=$Emitter.NormalizedLoopAge, B=1, Comparison Type=::NewEnumerator6)
    Emitter.Position:NiagaraPosition=&Multiply_VectorByFloat(
     Vector=&AddVectorToPosition(
      Position=$Engine.Owner.Position
      Vector=&MakeVector(
       X=0
       Y=0
       Z=&Multiply_Float(
        A=$Emitter.Global Scale
        B=&FloatFromCurve(
         FloatCurve=DataInterface() {propertyValues {Curve {keys [
             {arriveTangent=3.7918, interpMode=RCIM_Cubic, time=0, value=0.5}
             {arriveTangent=0.8971, interpMode=RCIM_Cubic, leaveTangent=0.8971, tangentMode=RCTM_User, time=2, value=1.7}
            ]}}}
         CurveIndex=$Engine.System.Age
         Scale Curve=200
        )
       )
      )
     )
     Float=1
    )
    Emitter.Radius:NiagaraFloat=&Multiply_Float(
     A=&Multiply_Float(
      A=570
      B=&FloatFromCurve(FloatCurve=DataInterface() {propertyValues {Curve {keys [<at 0 0.7>,<at 1.5 1.5>]}}}, CurveIndex=$Engine.System.Age, Scale Curve=1.3)
     )
     B=$Emitter.Global Scale
    )
    Emitter.VisibilityTag=1
   ) {isSetParameters=true}
  ]
  particleSpawnScript [&InitializeParticle(
    Mesh Scale=<12 12 12>
    Mesh Scale Mode=::NewEnumerator4
    Sprite Size Mode=::NewEnumerator3
    Uniform Sprite Size=500
    UsePositionOffset=false
    Write Material Random=true
   ) {module=InitializeParticle001}]
  particleUpdateScript [
   &ParticleState(Kill Particles When Lifetime Has Elapsed=false)
   /SetVariables(Particles.BlendWeight:NiagaraFloat=&FloatFromCurve(
     FloatCurve=DataInterface() {propertyValues {Curve {keys [<at 0 0 RCIM_Cubic>,<at 1 1 RCIM_Cubic>]}}}
     CurveIndex=$System.Timeline.NormalizedTime
    )) {isSetParameters=true}
  ]
  rendererClasses [&NiagaraComponentRendererProperties]
  renderers [{rendererClass=&NiagaraComponentRendererProperties, rendererIndex=0}]
 }
}
summary {
 systemName=NS_FNM_GnatGala_Vignette
 emitters [
  {emitterName=Minimal, rendererClasses [&NiagaraSpriteRendererProperties]}
  {bEnabled=false, emitterName=PP_Vignette, rendererClasses [&NiagaraComponentRendererProperties]}
 ]
 userVariables [
  User.FadeInOut {type=NiagaraBool, defaultValue=true}
  User.FOV {type=NiagaraFloat, defaultValue=0, description=""}
  User.Global Scale {type=NiagaraFloat, defaultValue=1, description=""}
  User.Ratio {type=Vector2f, description="", defaultValue {struct=Vector2f, value=<1.7778 1>}}
 ]
}
